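; Player gameplay ability for Starsoap that grants a bonus loot drop, apparently for eliminations.
; On activation it spawns loot from the LTG_Starsoap_Activity table at the avatar's location.
; It also broadcasts a gameplay alert through the global event router, throttled by real time against a scalable-float minimum interval and the stored last alert time.
; The failed-activation and end-ability overrides only call their parents.
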(event Ability|EventFailedToActivatePassiveAbility
  (params (ActorInfo :type Gameplay%20Ability%20Actor%20Info%20Structure))
  (|CallParentFunction ActorInfo))

(event Ability|EventActivateAbility
  (bind _returnvalue (Utilities|Time|GetRealTimeSeconds))
  (|CallParentFunction)
  (Class|AthenaFunctionLibrary|SpawnLootfromTableUsingItemEntry "LTG_Starsoap_Activity" 1 -1 (Transformation|GetActorLocation (Ability|GetAvatarActorfromActorInfo)) "0.000000,0.000000,1.000000" 0 true true)
  (if (> (- _returnvalue (Variables|Default|GetLastGameplayAlertTime)) (ScalableFloat|GetValueatLevel (Variables|Default|GetMinTimeBetweenGameplayAlerts)))
    (GameplayEventRouter|BroadcastStatefulEvent (GameplayEventRouter|GetGlobalEventRouter) (Variables|Default|GetGameplayAlert) 0)
    (Variables|Default|SetLastGameplayAlertTime _returnvalue)))

(event Ability|EventOnEndAbility
  (params (bWasCancelled :type Boolean))
  (|CallParentFunction bWasCancelled))
