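; Passive Starsoap ability, granted by a gameplay effect volume, that manages a player's weapon slot and gameplay alerts while in the area.
; On entry with a non-harvesting weapon, it switches to primary slot 0 unless the volume's target is active, then shows a delayed, state-based alert.
; Non-pickaxe swaps update the saved slot and can re-alert; ending restores that slot.
; BroadcastAlert replaces the previous alert and sets or clears stateful target data.
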(event Custom|InitialUIPopup
  (bind _returnvalue (Ability|GetGrantedbyEffectContext))
  (bind _returnvalue_2 (Ability|EffectContext|GetOriginalInstigatorActor _returnvalue))
  (bind _returnvalue_3 (Utilities|Time|GetRealTimeSeconds))
  (bind _gameplayalert_warning (Variables|Default|GetGameplayAlert_Warning))
  (bind _gameplayalert_warmup (Variables|Default|GetGameplayAlert_Warmup))
  (bind _gameplayalert_targetactive (Variables|Default|GetGameplayAlert_TargetActive))
  ; UI Popup on State Change
  (Utilities|FlowControl|Delay (ScalableFloat|GetValueatLevel (Variables|Default|GetUI_Delay)))
  (Utilities|IsValid _returnvalue_2
    (:Is%20Valid
      (bind _asbga_starsoap_gameplay_effect_volume (Utilities|Casting|CastToBGA_Starsoap_GameplayEffectVolume _returnvalue_2))
      (Networking|CanExecuteCosmeticEvents
        (:True
          (if (> (- _returnvalue_3 (Variables|Default|GetLastGameplayAlertTime)) (ScalableFloat|GetValueatLevel (Variables|Default|GetMinTimeBetweenGameplayAlerts)))
            (if (Utilities|Enum|Equal(Enum) (Variables|/StarsoapGameplay/Gameplay/BGA_Starsoap_GameplayEffectVolume.BGA_Starsoap_GameplayEffectVolume_C|GetStarsoap_States _asbga_starsoap_gameplay_effect_volume) "NewEnumerator2")
              (if (== (Variables|Default|GetPlayer) (Variables|/StarsoapGameplay/Gameplay/BGA_Starsoap_GameplayEffectVolume.BGA_Starsoap_GameplayEffectVolume_C|GetCurrentTarget _asbga_starsoap_gameplay_effect_volume))
                (CallFunction|BroadcastAlert (Variables|Default|GetGameplayAlert_IsTarget) (select (Utilities|Enum|Equal(Enum) (Variables|/StarsoapGameplay/Gameplay/BGA_Starsoap_GameplayEffectVolume.BGA_Starsoap_GameplayEffectVolume_C|GetStarsoap_States _asbga_starsoap_gameplay_effect_volume) "NewEnumerator2") (Variables|/Script/Engine.Pawn|GetPlayerState (Variables|/StarsoapGameplay/Gameplay/BGA_Starsoap_GameplayEffectVolume.BGA_Starsoap_GameplayEffectVolume_C|GetCurrentTarget _asbga_starsoap_gameplay_effect_volume)) 0))
                (Variables|Default|SetLastGameplayAlertTime _returnvalue_3)
                (else
                  (CallFunction|BroadcastAlert (select (Utilities|Enum|Equal(Enum) (Team|GetTeamAffiliation (Variables|/StarsoapGameplay/Gameplay/BGA_Starsoap_GameplayEffectVolume.BGA_Starsoap_GameplayEffectVolume_C|GetCurrentTarget _asbga_starsoap_gameplay_effect_volume) (Ability|GetAvatarActorfromActorInfo)) "Friendly") (Variables|Default|GetGameplayAlert_IsTargetTeammate) (Utilities|Select :Index (Variables|/StarsoapGameplay/Gameplay/BGA_Starsoap_GameplayEffectVolume.BGA_Starsoap_GameplayEffectVolume_C|GetStarsoap_States _asbga_starsoap_gameplay_effect_volume) :NewEnumerator0 _gameplayalert_warmup :NewEnumerator1 _gameplayalert_warning :NewEnumerator2 _gameplayalert_targetactive)) (select (Utilities|Enum|Equal(Enum) (Variables|/StarsoapGameplay/Gameplay/BGA_Starsoap_GameplayEffectVolume.BGA_Starsoap_GameplayEffectVolume_C|GetStarsoap_States _asbga_starsoap_gameplay_effect_volume) "NewEnumerator2") (Variables|/Script/Engine.Pawn|GetPlayerState (Variables|/StarsoapGameplay/Gameplay/BGA_Starsoap_GameplayEffectVolume.BGA_Starsoap_GameplayEffectVolume_C|GetCurrentTarget _asbga_starsoap_gameplay_effect_volume)) 0))
                  (Variables|Default|SetLastGameplayAlertTime _returnvalue_3)))
              (else
                (CallFunction|BroadcastAlert (Utilities|Select :Index (Variables|/StarsoapGameplay/Gameplay/BGA_Starsoap_GameplayEffectVolume.BGA_Starsoap_GameplayEffectVolume_C|GetStarsoap_States _asbga_starsoap_gameplay_effect_volume) :NewEnumerator0 _gameplayalert_warmup :NewEnumerator1 _gameplayalert_warning :NewEnumerator2 _gameplayalert_targetactive) (select (Utilities|Enum|Equal(Enum) (Variables|/StarsoapGameplay/Gameplay/BGA_Starsoap_GameplayEffectVolume.BGA_Starsoap_GameplayEffectVolume_C|GetStarsoap_States _asbga_starsoap_gameplay_effect_volume) "NewEnumerator2") (Variables|/Script/Engine.Pawn|GetPlayerState (Variables|/StarsoapGameplay/Gameplay/BGA_Starsoap_GameplayEffectVolume.BGA_Starsoap_GameplayEffectVolume_C|GetCurrentTarget _asbga_starsoap_gameplay_effect_volume)) 0))
                (Variables|Default|SetLastGameplayAlertTime _returnvalue_3)))))))))

(event Ability|EventFailedToActivatePassiveAbility
  (params (ActorInfo :type Gameplay%20Ability%20Actor%20Info%20Structure))
  (|CallParentFunction ActorInfo))

(event Ability|EventActivateAbility
  (bind _player (Variables|Default|GetPlayer))
  (|CallParentFunction)
  (Utilities|FlowControl|Sequence
    (:then_0
      (bind _asfort_player_pawn_athena (Utilities|Casting|CastToFortPlayerPawnAthena (Ability|GetAvatarActorfromActorInfo)))
      (Variables|Default|SetPlayer _asfort_player_pawn_athena)
      (Utilities|IsValid (Fortnite|GetFortPlayerControllerfromActor _asfort_player_pawn_athena)
        (:Is%20Valid
          (Variables|Default|SetFocusedSlot (Quickbars|GetFocusedSlot (Fortnite|GetFortPlayerControllerfromActor _asfort_player_pawn_athena) "Primary")))
        (:Is%20Not%20Valid
          _))
      (Utilities|IsValid (Variables|/Script/FortniteGame.FortPawn|GetCurrentWeapon _asfort_player_pawn_athena)
        (:Is%20Valid
          (Variables|Default|SetWeaponOnEnter (Variables|/Script/FortniteGame.FortPawn|GetCurrentWeapon _asfort_player_pawn_athena))
          (if (not (Tags|WeaponDataHasAnyMatchingGameplayTags (Variables|/Script/FortniteGame.FortPawn|GetCurrentWeapon _asfort_player_pawn_athena) "(GameplayTags=((TagName=\"Granted.ItemEquipped.Melee.Harvest\")))"))
            (bind _asbga_starsoap_gameplay_effect_volume (Utilities|Casting|CastToBGA_Starsoap_GameplayEffectVolume (Ability|EffectContext|GetInstigatorActor (Ability|GetGrantedbyEffectContext))))
            (if (not (Utilities|Enum|Equal(Enum) (Variables|/StarsoapGameplay/Gameplay/BGA_Starsoap_GameplayEffectVolume.BGA_Starsoap_GameplayEffectVolume_C|GetStarsoap_States _asbga_starsoap_gameplay_effect_volume) "NewEnumerator2"))
              (Quickbars|ClientActivateSlot (Fortnite|GetFortPlayerControllerfromActor _player) "Primary" 0 0.0 true false)
              (FortPawn|BindEventtoOnWeaponEquippedDelegate _player (EventDispatchers|CreateEvent :!event Custom|WeaponSwapped)))
            (CallFunction|InitialUIPopup)))
        (:Is%20Not%20Valid
          _)))
    (:then_1)))

(event Ability|EventOnEndAbility
  (params (bWasCancelled :type Boolean))
  (bind _player (Variables|Default|GetPlayer))
  (|CallParentFunction bWasCancelled)
  (Utilities|FlowControl|Sequence
    (:then_0
      (FortPawn|UnbindEventfromOnWeaponEquippedDelegate _player (EventDispatchers|CreateEvent :!event Custom|WeaponSwapped)))
    (:then_1
      (if (not (FortPawn|IsDead _player))
        (Quickbars|ClientActivateSlot (Fortnite|GetFortPlayerControllerfromActor _player) "Primary" (Variables|Default|GetFocusedSlot) 0.1 true false)))))

(event Custom|WeaponSwapped
  (params (NewWeapon :type Fort%20Weapon%20Object%20Reference) (PrevWeapon :type Fort%20Weapon%20Object%20Reference))
  (bind _returnvalue (Utilities|Time|GetRealTimeSeconds))
  (if (not (or (Utilities|IsA NewWeapon "/Script/FortniteGame.FortWeaponPickaxeAthena") (Utilities|IsA NewWeapon "/Script/FortniteGame.FortWeaponPickaxeDualWieldAthena")))
    (Variables|Default|SetFocusedSlot (Quickbars|GetFocusedSlot (Fortnite|GetFortPlayerControllerfromActor (Variables|Default|GetPlayer)) "Primary"))
    (bind _asbga_starsoap_gameplay_effect_volume (Utilities|Casting|CastToBGA_Starsoap_GameplayEffectVolume (Ability|EffectContext|GetInstigatorActor (Ability|GetGrantedbyEffectContext))))
    (Utilities|IsValid (Variables|/StarsoapGameplay/Gameplay/BGA_Starsoap_GameplayEffectVolume.BGA_Starsoap_GameplayEffectVolume_C|GetCurrentTarget _asbga_starsoap_gameplay_effect_volume)
      (:Is%20Not%20Valid
        (if (GameplayTags|HasAnyMatchingGameplayTags (Variables|Default|GetPlayer) (Variables|Default|GetT_InArea))
          (if (> (- _returnvalue (Variables|Default|GetLastGameplayAlertTime)) (ScalableFloat|GetValueatLevel (Variables|Default|GetMinTimeBetweenGameplayAlerts)))
            (CallFunction|BroadcastAlert (Variables|Default|GetGameplayAlert))
            (Variables|Default|SetLastGameplayAlertTime _returnvalue)))))))

(event Custom|TestQuickAlerts
  (Utilities|FlowControl|Delay 2.0)
  (CallFunction|BroadcastAlert (Variables|Default|GetGameplayAlert))
  (Utilities|FlowControl|Delay 2.0)
  (CallFunction|BroadcastAlert (Variables|Default|GetGameplayAlert_TargetActive) (Variables|/Script/Engine.Pawn|GetPlayerState (Variables|Default|GetPlayer)))
  (Utilities|FlowControl|Delay 2.0)
  (CallFunction|BroadcastAlert (Utilities|Struct|MakeFortEventAddGameplayAlertRequested "(TagName=\"Starsoap.TargetEliminated\")") 0 true))

(fn BroadcastAlert
  (params (InGameplayAlert :type Fort%20Event%20Add%20Gameplay%20Alert%20Requested%20Structure) (InPlayerState :type Player%20State%20Object%20Reference) (InIsEliminated :type Boolean))
  (bind _returnvalue (GameplayEventRouter|GetGlobalEventRouter))
  (bind _returnvalue_2 (GameplayEventRouter|GetGlobalEventRouter))
  (bind _lastbroadcastalerttag (Variables|Default|GetLastBroadcastAlertTag))
  (bind _targetplayerstate (LocalVariables|GetTargetPlayerState))
  (Utilities|FlowControl|Sequence
    (:then_0
      (if (GameplayTags|IsGameplayTagValid _lastbroadcastalerttag)
        (GameplayEventRouter|BroadcastEvent _returnvalue (Utilities|Struct|MakeFortEventRemoveGameplayAlertRequested _lastbroadcastalerttag) 0)
        (GameplayEventRouter|ClearEventState _returnvalue "/Script/BattleRoyaleGameplayAlerts.FortEvent_CurrentGameplayAlertChanged")
        (Variables|Default|SetLastBroadcastAlertTag)))
    (:then_1
      (bind _asfort_player_state_athena (Utilities|Casting|CastToFortPlayerStateAthena InPlayerState))
      (LocalVariables|SetTargetPlayerState _asfort_player_state_athena))
    (:then_2
      (bind _output_get (Variables|Default|SetLastBroadcastAlertTag (Utilities|Struct|BreakFortEventAddGameplayAlertRequested (Variables|GetInGameplayAlert))))
      (if (GameplayTags|IsGameplayTagValid _output_get)
        (GameplayEventRouter|BroadcastStatefulEvent (GameplayEventRouter|GetGlobalEventRouter) InGameplayAlert 0)
        (else
          _))
      (if (or (Utilities|IsValid _targetplayerstate :!class /Script/Engine.KismetSystemLibrary) InIsEliminated)
        (GameplayEventRouter|BroadcastStatefulEvent _returnvalue_2 (Starsoap|TargetData|MakeStarsoapTargetData _targetplayerstate InIsEliminated) 0)
        (else
          (GameplayEventRouter|ClearEventState _returnvalue_2 "/Script/StarsoapGameplay.StarsoapTargetData"))))))
