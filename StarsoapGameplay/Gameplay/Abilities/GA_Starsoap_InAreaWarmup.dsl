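; Gameplay ability for a player in a Starsoap effect volume that restricts weapon swapping and shows a state-based alert.
; On activation it saves the focused primary quickbar slot, switches slot if a weapon is held, and applies a swap-blocking effect.
; It broadcasts a warmup, warning or target-active alert from the instigating volume's state, throttled by a minimum interval.
; On end it removes the effect and restores the saved slot unless the player is dead.
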(event Custom|InitialUIPopup
  (bind _returnvalue (Ability|GetGrantedbyEffectContext))
  (bind _returnvalue_2 (Ability|EffectContext|GetOriginalInstigatorActor _returnvalue))
  (bind _returnvalue_3 (Utilities|Time|GetRealTimeSeconds))
  (Utilities|IsValid _returnvalue_2
    (:Is%20Valid
      (bind _asbga_starsoap_gameplay_effect_volume (Utilities|Casting|CastToBGA_Starsoap_GameplayEffectVolume _returnvalue_2))
      (if (> (- _returnvalue_3 (Variables|Default|GetLastGameplayAlertTime)) (ScalableFloat|GetValueatLevel (Variables|Default|GetMinTimeBetweenGameplayAlerts)))
        (GameplayEventRouter|BroadcastStatefulEvent (GameplayEventRouter|GetGlobalEventRouter) (Utilities|Select :Index (Variables|/StarsoapGameplay/Gameplay/BGA_Starsoap_GameplayEffectVolume.BGA_Starsoap_GameplayEffectVolume_C|GetStarsoap_States _asbga_starsoap_gameplay_effect_volume) :NewEnumerator0 (Variables|Default|GetGameplayAlert_Warmup) :NewEnumerator1 (Variables|Default|GetGameplayAlert_Warning) :NewEnumerator2 (Variables|Default|GetGameplayAlert_TargetActive)) 0)
        (Variables|Default|SetLastGameplayAlertTime _returnvalue_3)))))

(event Ability|EventFailedToActivatePassiveAbility
  (params (ActorInfo :type Gameplay%20Ability%20Actor%20Info%20Structure))
  (|CallParentFunction ActorInfo))

(event Ability|EventActivateAbility
  (bind _player (Variables|Default|GetPlayer))
  (|CallParentFunction)
  (bind _asfort_player_pawn_athena (Utilities|Casting|CastToFortPlayerPawnAthena (Ability|GetAvatarActorfromActorInfo)))
  (Variables|Default|SetPlayer _asfort_player_pawn_athena)
  (Utilities|IsValid (Fortnite|GetFortPlayerControllerfromActor _asfort_player_pawn_athena)
    (:Is%20Valid
      (Variables|Default|SetFocusSlot (Quickbars|GetFocusedSlot (Fortnite|GetFortPlayerControllerfromActor _asfort_player_pawn_athena) "Primary")))
    (:Is%20Not%20Valid
      _))
  (Utilities|IsValid (Variables|/Script/FortniteGame.FortPawn|GetCurrentWeapon _asfort_player_pawn_athena)
    (:Is%20Valid
      (Fortnite|ActivateQuickBarSlot _player))
    (:Is%20Not%20Valid
      _))
  (GameplayEffects|ApplyGameplayEffectToSelf :self (Ability|GetAbilitySystemComponent _player) :GameplayEffectClass (Variables|Default|GetGE_BlockWeaponSwap) :Level 0.0)
  (CallFunction|InitialUIPopup))

(event Ability|EventOnEndAbility
  (params (bWasCancelled :type Boolean))
  (bind _player (Variables|Default|GetPlayer))
  (bind _returnvalue (Fortnite|GetFortPlayerControllerfromActor _player))
  (|CallParentFunction bWasCancelled)
  (Utilities|FlowControl|Sequence
    (:then_0
      (GameplayEffects|RemoveActiveGameplayEffectbySourceEffect (Ability|GetAbilitySystemComponent _player) (Variables|Default|GetGE_BlockWeaponSwap) 0 -1))
    (:then_1
      (if (not (FortPawn|IsDead _player))
        (Utilities|IsValid _returnvalue
          (:Is%20Valid
            (Quickbars|ClientActivateSlot _returnvalue "Primary" (Variables|Default|GetFocusSlot) 0.1 true false)))))))
