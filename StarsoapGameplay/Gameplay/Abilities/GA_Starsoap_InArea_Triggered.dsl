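; Event-triggered Starsoap ability that reports player eliminations to the gameplay effect volume that granted it.
; On activation it runs the parent logic, then continues only if the event's target is a player pawn without any excluded tags.
; It finds the volume as instigator of the granting effect's context and, if that volume is in a specific state, calls PlayerEliminated with the avatar and target pawns.
; The ability then ends.
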(event Ability|EventActivateAbilityFromEvent
  (params (EventData :type Gameplay%20Event%20Data%20Structure%20%28by%20ref%29))
  (bind (_eventtag _instigator _target _optionalobject _optionalobject2 _contexthandle _instigatortags _targettags) (Utilities|Struct|BreakGameplayEventData EventData))
  (|CallParentFunction EventData)
  (if (not (GameplayTags|HasAnyTags _targettags (Variables|Default|GetOther%20Container)))
    (bind _asfort_player_pawn (Utilities|Casting|CastToFortPlayerPawn _target)
      (:then
        (bind _asbga_starsoap_gameplay_effect_volume (Utilities|Casting|CastToBGA_Starsoap_GameplayEffectVolume (Ability|EffectContext|GetInstigatorActor (Ability|GetGrantedbyEffectContext)))
          (:then
            (if (Utilities|Enum|Equal(Enum) (Variables|/StarsoapGameplay/Gameplay/BGA_Starsoap_GameplayEffectVolume.BGA_Starsoap_GameplayEffectVolume_C|GetStarsoap_States _asbga_starsoap_gameplay_effect_volume) "NewEnumerator1")
              (Class|BGAStarsoapGameplayEffectVolume|PlayerEliminated _asbga_starsoap_gameplay_effect_volume (CallFunction|GetAvatarasFortPlayerPawn) _asfort_player_pawn)
              (else
                _)))
          (:CastFailed
            _))
        (Ability|EndAbility))
      (:CastFailed
        _)))
  (Ability|EndAbility))
