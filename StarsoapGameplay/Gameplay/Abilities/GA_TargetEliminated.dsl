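; Gameplay ability for target elimination that broadcasts a gameplay alert, throttled to a minimum interval.
; On activation it compares real time with the last alert time against a scalable-float minimum gap.
; When enough time has passed, it sends the configured alert and a newly made Starsoap target data payload as stateful events through the global event router, then records the time.
; Failed passive activation and ability end only defer to the parent.
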(event Ability|EventFailedToActivatePassiveAbility
  (params (ActorInfo :type Gameplay%20Ability%20Actor%20Info%20Structure))
  (|CallParentFunction ActorInfo))

(event Ability|EventActivateAbility
  (bind _returnvalue (GameplayEventRouter|GetGlobalEventRouter))
  (bind _returnvalue_2 (Utilities|Time|GetRealTimeSeconds))
  (|CallParentFunction)
  (if (> (- _returnvalue_2 (Variables|Default|GetLastGameplayAlertTime)) (ScalableFloat|GetValueatLevel (Variables|Default|GetMinTimeBetweenGameplayAlerts)))
    (GameplayEventRouter|BroadcastStatefulEvent _returnvalue (Variables|Default|GetGameplayAlert) 0)
    (GameplayEventRouter|BroadcastStatefulEvent _returnvalue (Starsoap|TargetData|MakeStarsoapTargetData 0 true) 0)
    (Variables|Default|SetLastGameplayAlertTime _returnvalue_2)))

(event Ability|EventOnEndAbility
  (params (bWasCancelled :type Boolean))
  (|CallParentFunction bWasCancelled))
